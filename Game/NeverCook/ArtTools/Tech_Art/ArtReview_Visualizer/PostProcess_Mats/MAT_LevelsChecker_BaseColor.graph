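# Post-process art review material that flags out-of-range base color, metallic reflectance and specular values on screen.
# Nested Ifs test stored specular, desaturated base color extremes, and brightest channel against a metallic-weighted minimum, tinting failures by warning color, with an animated zebra-stripe mask defined for highlighting.
# A depth-based SphereMask limits warnings to nearby pixels, and a static switch optionally ignores subsurface.
# An aspect-corrected legend texture overlays the screen.
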Material /Game/NeverCook/ArtTools/Tech_Art/ArtReview_Visualizer/PostProcess_Mats/MAT_LevelsChecker_BaseColor
graph {
 properties {materialDomain=MD_PostProcess, pixelDepthOffsetMode=PDOM_Legacy, usedShadingModels=Default Lit}
 nodes [
  Constant3Vector() {constant=<rgba 0 1 0 0>}
  Constant3Vector() {constant=<rgba 1 0 0 0>}
  n1 = Constant()
  n2 = Constant() {r=1}
  If(
   A=ComponentMask(None=SceneTexture() {sceneTextureId=PPI_ShadingModelID}.Color) {r=true}
   B=ScalarParameter() {defaultValue=2}
   "A > B"=@n1
   "A == B"=@n2
   "A < B"=@n1
  ) {constB=6, desc=Doesn't detect all subsurface}
  n3 = SceneTexture() {sceneTextureId=PPI_SceneDepth}
  n4 = SceneTexture() {sceneTextureId=PPI_StoredBaseColor}
  n5 = MaterialFunctionCall(Float3=@n4.Color) {materialFunction=&BreakOutFloat3Components}
  n6 = Max(A=Max(A=@n5.R, B=@n5.R), B=@n5.B)
  n7 = ComponentMask(None=SceneTexture() {sceneTextureId=PPI_PostProcessInput0}.Color) {b=true, g=true, r=true}
  n8 = MaterialFunctionCall(Float3=Desaturation(None=@n4.Color, Fraction=Constant() {r=1})) {materialFunction=&BreakOutFloat3Components}
  n9 = If(
   A=@n8.R
   B=ScalarParameter() {defaultValue=0.9, parameterName=MaxValue}
   "A > B"=Multiply(A=VectorParameter() {parameterName=MaxValueColor, defaultValue=<rgba 1 0 0.8314 1>}.RGB, B=NamedRerouteUsage())
   "A == B"=@n7
   "A < B"=@n7
  )
  n10 = SceneTexture() {sceneTextureId=PPI_Metallic}
  n11 = LinearInterpolate(
   A=ScalarParameter() {defaultValue=0.009, parameterName=MinValue}
   B=ScalarParameter() {defaultValue=0.25, parameterName=MinMetallicValue}
   Alpha=Power(Base=ComponentMask(None=@n10.Color) {r=true}) {constExponent=3}
  )
  n12 = If(
   A=@n6
   B=@n11
   "A > B"=@n9
   "A == B"=@n9
   "A < B"=Multiply(A=VectorParameter() {parameterName=MinMetallicValueColor, defaultValue=<rgba 1 0 0 1>}.RGB, B=NamedRerouteUsage())
  )
  n13 = If(
   A=@n8.R
   B=ScalarParameter() {defaultValue=0.009, parameterName=MinValue}
   "A > B"=@n12
   "A == B"=@n12
   "A < B"=Multiply(A=VectorParameter() {parameterName=MinValueColor, defaultValue=<rgba 1 1 0 1>}.RGB, B=NamedRerouteUsage())
  )
  n14 = If(
   A=ComponentMask(None=SceneTexture() {sceneTextureId=PPI_StoredSpecular}.Color) {r=true}
   B=ScalarParameter() {defaultValue=0.2, parameterName=MinSpecular}
   "A > B"=@n13
   "A == B"=@n13
   "A < B"=Multiply(A=VectorParameter() {parameterName=MinSpecularValueColor, defaultValue=<rgba 0.03 0.03 1 0>}.RGB, B=NamedRerouteUsage())
  )
  n15 = LinearInterpolate(
   A=@n7
   B=@n14
   Alpha=SphereMask(
    A=ComponentMask(None=@n3.Color) {r=true}
    B=Constant()
    Radius=ScalarParameter() {defaultValue=3000, parameterName=MaxDistance}
    Hardness=Constant() {r=100}
   )
  )
  n16 = NamedRerouteUsage()
  n17 = NamedRerouteUsage()
  LinearInterpolate(
   A=StaticSwitchParameter(True=LinearInterpolate(A=@n15, B=@n16, Alpha=@n17), False=@n15) {parameterName=UseSubsurfaceIgnore}
   B=NamedRerouteUsage()
   Alpha=NamedRerouteUsage()
  )
  n18 = ViewProperty() {property=MEVP_ViewSize}
  n19 = Divide(A=ComponentMask(None=@n18.Property) {r=true}, B=ComponentMask(None=@n18.Property) {g=true})
  n20 = ScreenPosition()
  n21 = ComponentMask(None=@n20.ViewportUV) {r=true}
  n22 = Add(A=ComponentMask(None=@n20.ViewportUV) {g=true}) {constB=-0.825}
  n23 = Multiply(A=MaterialFunctionCall(X=Multiply(A=@n19, B=@n21), Y=Multiply(A=@n22) {constB=2}) {materialFunction=&MakeFloat2}.Result) {constB=2.75}
  NamedRerouteDeclaration(None=TextureSample(UVs=@n23) {texture=&/T_LevelCheckerLegend}.RGB) {name=Legend, nodeColor=<rgba 0.2891 0 1 1>}
  n24 = ComponentMask(None=@n23) {r=true}
  n25 = ComponentMask(None=@n23) {g=true}
  NamedRerouteDeclaration(None=Multiply(A=Step(Y=@n24) {constY=1}, B=Step(X=@n25))) {name=LegendMask, nodeColor=<rgba 0 0.8359 1 1>}
  n26 = ScreenPosition()
  n27 = Add(A=Multiply(A=ComponentMask(None=@n26.PixelPosition) {r=true}) {constB=-1}, B=ComponentMask(None=@n26.PixelPosition) {g=true})
  n28 = ScalarParameter() {defaultValue=10, parameterName=Zebra Stripes Size}
  n29 = Time()
  n30 = ScalarParameter() {defaultValue=1.5, parameterName=Zebra Stripes Speed}
  n31 = Sine(None=Add(A=Divide(A=@n27, B=@n28) {constB=20}, B=Multiply(A=@n29, B=@n30)))
  NamedRerouteDeclaration(None=LinearInterpolate(Alpha=SmoothStep(Value=@n31) {constMin=-1}) {constA=0.7}) {name=Zebra Stripes, nodeColor=<rgba 1 0.2344 0 1>}
  n32 = SceneTexture() {sceneTextureId=PPI_SubsurfaceColor}
  NamedRerouteDeclaration(None=If(
    A=MaterialFunctionCall(f3=@n32.Color) {materialFunction=&AddComponents}.f3
    "A > B"=@n2
    "A == B"=@n1
    "A < B"=@n1
   )) {name=Subsurface masking, nodeColor=<rgba 1 0 0.3516 1>}
  NamedRerouteDeclaration(None=@n7) {name=Lit, nodeColor=<rgba 1 0 0.0937 1>}
  Reroute(None=@n9) {desc=MaxValueCheck}
  Reroute(None=@n12) {desc=MetallicCheck}
  Reroute(None=@n14) {desc=Warnings}
  Reroute(None=@n8.R) {desc=Luminance}
  Reroute(None=@n6) {desc=reflectance}
  Reroute(None=@n11) {desc=adjustedMinValue}
 ]
}
