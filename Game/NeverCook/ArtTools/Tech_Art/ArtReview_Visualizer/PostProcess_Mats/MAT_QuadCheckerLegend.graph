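# Post-process material for the art review visualizer that overlays the quad checker legend texture in the lower-left corner of the screen.
# Viewport UVs are offset, scaled and corrected by the view's aspect ratio to place the legend without distortion as the view shape changes.
# Step masks combine with texture alpha to confine the legend to one tile.
# Named reroutes carry legend color and mask into a lerp over scene color.
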Material /Game/NeverCook/ArtTools/Tech_Art/ArtReview_Visualizer/PostProcess_Mats/MAT_QuadCheckerLegend
graph {
 properties {materialDomain=MD_PostProcess, pixelDepthOffsetMode=PDOM_Legacy, usedShadingModels=Default Lit}
 nodes [
  LinearInterpolate(
   A=ComponentMask(None=SceneTexture() {sceneTextureId=PPI_PostProcessInput0}.Color) {b=true, g=true, r=true}
   B=NamedRerouteUsage()
   Alpha=NamedRerouteUsage()
  )
  n1 = ViewProperty() {property=MEVP_ViewSize}
  n2 = Divide(A=ComponentMask(None=@n1.Property) {r=true}, B=ComponentMask(None=@n1.Property) {g=true})
  n3 = ScreenPosition()
  n4 = ComponentMask(None=@n3.ViewportUV) {r=true}
  n5 = Add(A=ComponentMask(None=@n3.ViewportUV) {g=true}) {constB=-0.825}
  n6 = Multiply(A=MaterialFunctionCall(X=Multiply(A=@n2, B=@n4), Y=Multiply(A=@n5) {constB=2}) {materialFunction=&MakeFloat2}.Result) {constB=2.75}
  n7 = TextureSample(UVs=@n6) {texture=&/T_QuadCheckerLegend}
  NamedRerouteDeclaration(None=@n7.RGB) {name=Legend, nodeColor=<rgba 0.2891 0 1 1>}
  n8 = Step(Y=ComponentMask(None=@n6) {r=true}) {constY=1}
  n9 = Step(X=ComponentMask(None=@n6) {g=true})
  NamedRerouteDeclaration(None=Multiply(A=@n7.A, B=Multiply(A=@n8, B=@n9))) {name=LegendMask, nodeColor=<rgba 0 0.8359 1 1>}
 ]
}
