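# MF_InvalidBufferScreen is a material function in the ArtReview Visualizer post-process tools that outputs a screen-space tiled placeholder image from T_InvalidBuffer.
# It centers the viewport UVs, scales them by visible resolution divided by its height to correct for aspect ratio, then restores the offset.
# The corrected UVs are multiplied to tile the texture, whose RGB is the function's output.
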MaterialFunction /Game/NeverCook/ArtTools/Tech_Art/ArtReview_Visualizer/PostProcess_Mats/MF_InvalidBufferScreen
graph {nodes [
  n1 = Add(A=ScreenPosition().ViewportUV) {constB=-0.5}
  n2 = MaterialFunctionCall() {materialFunction=&ScreenResolution}
  n3 = Add(A=Divide(A=Multiply(A=@n1, B=@n2.Visible Resolution), B=ComponentMask(None=@n2.Visible Resolution) {g=true})) {constB=0.5}
  FunctionOutput(None=TextureSample(UVs=Multiply(A=@n3) {constB=6}) {texture=&/T_InvalidBuffer}.RGB)
 ]}
