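# Post-process material for the art review visualizer that shows the stored base color buffer as a desaturated grayscale view.
# It desaturates the scene's stored base color, then applies a power curve that approximates linear-to-gamma conversion.
# Feature level and shading path switches substitute the MF_InvalidBufferScreen function output on mobile feature levels and forward shading.
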Material /Game/NeverCook/ArtTools/Tech_Art/ArtReview_Visualizer/PostProcess_Mats/MAT_artReview_BaseColorDesaturated
graph {
 properties {materialDomain=MD_PostProcess, pixelDepthOffsetMode=PDOM_Legacy, usedShadingModels=Default Lit}
 nodes [
  n1 = Desaturation(None=SceneTexture() {sceneTextureId=PPI_StoredBaseColor}.Color)
  n2 = MaterialFunctionCall() {materialFunction=&/MF_InvalidBufferScreen}
  n3 = FeatureLevelSwitch(Default=@n1, ES2=@n2.Result, ES3_1=@n2.Result, SM4_REMOVED=@n1, SM5=@n1, SM6=@n1)
  Power(Base=ShadingPathSwitch(Default=@n3, Deferred=@n3, Forward=@n2.Result)) {constExponent=0.4545}
 ]
}
