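# Post-process material for the ArtReview visualizer that displays the material ambient occlusion buffer for art review.
# A SceneTexture node reads MaterialAO, and a feature level switch swaps in MF_InvalidBufferScreen on ES2 and ES3_1.
# A shading path switch keeps that result for deferred rendering and shows the invalid-buffer screen under forward shading.
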Material /Game/NeverCook/ArtTools/Tech_Art/ArtReview_Visualizer/PostProcess_Mats/MAT_artReview_MaterialAmbientOcclusion
graph {
 properties {blendableLocation=BL_SceneColorAfterDOF, materialDomain=MD_PostProcess, pixelDepthOffsetMode=PDOM_Legacy, usedShadingModels=Default Lit}
 nodes [
  n1 = SceneTexture() {sceneTextureId=PPI_MaterialAO}
  n2 = MaterialFunctionCall() {materialFunction=&/MF_InvalidBufferScreen}
  n3 = FeatureLevelSwitch(Default=@n1.Color, ES2=@n2.Result, ES3_1=@n2.Result, SM4_REMOVED=@n1.Color, SM5=@n1.Color, SM6=@n1.Color)
  ShadingPathSwitch(Default=@n3, Deferred=@n3, Forward=@n2.Result)
 ]
}
