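# Post-process material in the art review visualizer that displays the scene's roughness buffer on screen.
# It samples the roughness scene texture and converts it from linear to sRGB.
# Feature level and shading path switches substitute the MF_InvalidBufferScreen function on ES2, ES3.1 and forward shading, presumably because that buffer is unavailable there.
# No node is shown wired to a material output in this graph.
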Material /Game/NeverCook/ArtTools/Tech_Art/ArtReview_Visualizer/PostProcess_Mats/MAT_artReview_Roughness
graph {
 properties {materialDomain=MD_PostProcess, pixelDepthOffsetMode=PDOM_Legacy, usedShadingModels=Default Lit}
 nodes [
  n1 = MaterialFunctionCall!(Linear=SceneTexture() {sceneTextureId=PPI_Roughness}.Color) {materialFunction=/Engine/Functions/Engine_MaterialFunctions01/ImageAdjustment/LinearTosRGB}
  n2 = MaterialFunctionCall() {materialFunction=&/MF_InvalidBufferScreen}
  n3 = FeatureLevelSwitch(Default=@n1.sRGB, ES2=@n2.Result, ES3_1=@n2.Result, SM4_REMOVED=@n1.sRGB, SM5=@n1.sRGB, SM6=@n1.sRGB)
  ShadingPathSwitch(Default=@n3, Deferred=@n3, Forward=@n2.Result)
 ]
}
