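# Post-process material for the art review visualizer that visualizes the alpha channel of the separate translucency buffer.
# It samples that scene texture, masks out its alpha and raises it to a power set by the Exp scalar parameter.
# The default exponent approximates a linear-to-gamma curve, and the blendable location is scene color before depth of field.
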Material /Game/NeverCook/ArtTools/Tech_Art/ArtReview_Visualizer/PostProcess_Mats/MAT_artReview_SeperateAlpha
graph {
 properties {blendableLocation=BL_SceneColorBeforeDOF, materialDomain=MD_PostProcess, pixelDepthOffsetMode=PDOM_Legacy, usedShadingModels=Default Lit}
 nodes [Power(
   Base=ComponentMask(None=SceneTexture() {sceneTextureId=PPI_SeparateTranslucency}.Color) {a=true}
   Exp=ScalarParameter() {defaultValue=0.45, parameterName=Exp, sliderMax=1}
  ) {constExponent=0.45}]
}
