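# Post-process material for an art review visualizer that displays the scene's world normal buffer.
# It remaps the WorldNormal scene texture from signed to 0-1 range with a half scale and half offset, and is set to blend after depth of field.
# Feature level and shading path switches substitute the MF_InvalidBufferScreen function's output on ES2, ES3.1 and forward shading.
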Material /Game/NeverCook/ArtTools/Tech_Art/ArtReview_Visualizer/PostProcess_Mats/MAT_artReview_WorldNormal
graph {
 properties {blendableLocation=BL_SceneColorAfterDOF, materialDomain=MD_PostProcess, pixelDepthOffsetMode=PDOM_Legacy, usedShadingModels=Default Lit}
 nodes [
  n1 = Add(A=Multiply(A=SceneTexture() {sceneTextureId=PPI_WorldNormal}.Color) {constB=0.5}) {constB=0.5}
  n2 = MaterialFunctionCall() {materialFunction=&/MF_InvalidBufferScreen}
  n3 = FeatureLevelSwitch(Default=@n1, ES2=@n2.Result, ES3_1=@n2.Result, SM4_REMOVED=@n1, SM5=@n1, SM6=@n1)
  ShadingPathSwitch(Default=@n3, Deferred=@n3, Forward=@n2.Result)
 ]
}
