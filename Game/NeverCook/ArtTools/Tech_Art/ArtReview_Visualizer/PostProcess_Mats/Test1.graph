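# Test1 is a minimal Default Lit test material in the never-cook Art Review Visualizer tools, stored among its post-process materials.
# It uses material attributes, with a Set Material Attributes node whose Base Color takes the RGB of a single vector parameter.
# The graph has no textures, math or other logic.
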Material /Game/NeverCook/ArtTools/Tech_Art/ArtReview_Visualizer/PostProcess_Mats/Test1
graph {
 properties {bAutomaticallySetUsageInEditor=false, bUseMaterialAttributes=true, usedShadingModels=Default Lit}
 nodes [SetMaterialAttributes!(MaterialAttributes, Base Color=VectorParameter().RGB) {attributeSetTypes [#1]}]
}
