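# Test2 is a small Default Lit surface material, apparently a test asset, stored among the ArtReview Visualizer post-process materials.
# It holds a mid-grey vector parameter and two scalar constants, with no math, textures or shown connections between them.
# Despite its folder, nothing in it sets up post-process behavior, and automatic usage flagging in the editor is turned off.
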Material /Game/NeverCook/ArtTools/Tech_Art/ArtReview_Visualizer/PostProcess_Mats/Test2
graph {
 properties {bAutomaticallySetUsageInEditor=false, usedShadingModels=Default Lit}
 nodes [
  Constant() {r=0.8}
  Constant()
  VectorParameter() {defaultValue=<rgba 0.4573 0.4573 0.4573 0>}
 ]
}
